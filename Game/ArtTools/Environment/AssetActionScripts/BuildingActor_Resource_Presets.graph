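# Editor utility with asset actions that apply resource, density and physics presets to selected building actor Blueprints.
# Each action loops over the selection and runs embedded Python to edit class defaults.
# Wood/Stone/Metal setters call Unify, which aligns attribute keys, resource-rate rows and prop physics tags, then saves.
# Unify and Set_asDense use CheckParentClass whitelists to branch between prop and architecture; Medium and Set_asDense never save.
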EditorUtilityBlueprint /Game/ArtTools/Environment/AssetActionScripts/BuildingActor_Resource_Presets
graphs [
 CheckParentClass {nodes [
   n1 = CheckParentClass!() -> (then, inAssetData:Asset Data Structure) {is=title K2Node_FunctionEntry}
   > n2 = LoadBlueprintClass!(execute, AssetPath:String="To String (Name)"(InName=Break Asset Data(AssetData=@n1))) -> (then, ReturnValue:Object Class Reference)
   > Set l_currentBP_Class(l_currentBP_Class:Object Class Reference=@n2) -> (Output_Get:Object Class Reference)
   > n3 = For Each Loop with Break!(
    Exec:Exec
    Array:Array of Object Class References=Get propClassWhitelist() -> (propClassWhitelist:Array of Object Class References)
    Break:Exec=@n9.then
   ) -> (LoopBody:Exec, Array Element:Object Class Reference, Array Index:int, Completed:Exec)
   n4 = Branch(execute=@n3.Completed, Condition=Get l_isProp() -> (l_isProp:bool))
   > Return Node!(execute, isProp:bool=Get l_isProp(), isValid:bool=Get l_isValid() -> (l_isValid:bool))
   FortShippingLog!(execute=@n4.else, WorldContextObject:Object Reference, InString:String=not prop, bLogAsWarning:bool=false) -> (then)
   > n5 = For Each Loop with Break(Array=Get architectureClassWhitelist() -> (architectureClassWhitelist:Array of Object Class References), Break=@n7.then)
   Return Node(execute=@n5.Completed, isProp=Get l_isProp(), isValid=Get l_isValid())
   > n6 = Branch(Condition=ClassIsChildOf(TestClass=Get l_currentBP_Class() -> (l_currentBP_Class:Object Class Reference), ParentClass=@n5))
   > n7 = Set l_isValid(l_isValid:bool=true) -> (Output_Get:bool)
   FortShippingLog(execute=@n6.else, InString=Append(A="not child of ", B=Get Class Display Name(Class=@n5)), bLogAsWarning=false)
   n8 = Set l_isProp(execute=@n3.LoopBody, l_isProp:bool=ClassIsChildOf(TestClass=Get l_currentBP_Class(), ParentClass=@n3)) -> (Output_Get:bool)
   > Branch(Condition=@n8)
   > n9 = Set l_isValid(l_isValid=true)
   Get inAssetData() -> (inAssetData:Asset Data Structure)
  ]}
 DisablePhysicsComponent {nodes [
   DisablePhysicsComponent() {is=title K2Node_FunctionEntry}
   > n1 = GetSelectedAssetData()
   > n2 = For Each Loop(Array:Array of Asset Data Structures=@n1) -> (Array Element:Asset Data Structure)
   n3 = ToSoftObjectPath!(InAssetData:"Asset Data Structure (by ref)"=@n2) -> (ReturnValue_PathString:String)
   > Execute Python Script!(
    execute
    PythonScript="import unreal\r\n\r\n## BuildingProp.NoPhysics\r\nassetObject = unreal.load_object(None, asset_object_path)\r\ncdo = unreal.get_default_object(assetObject)\r\ncdo.set_editor_property(\"bCreatePhysicsObjectComponent\", False)"
    asset_object_path:String=Append(A=@n3, B=_C)
   ) -> (then, ReturnValue) {inputs [asset_object_path]}
   > SaveAsset!(execute, AssetToSave:String=GetBaseFilename(InPath=@n3, bRemovePath=false), bOnlyIfIsDirty:bool=true) -> (then, ReturnValue:bool)
  ]}
 EventGraph {links [], nodes []}
 Medium {nodes [
   Medium() {is=title K2Node_FunctionEntry}
   > n1 = GetSelectedAssetData()
   > n2 = For Each Loop(Array=@n1)
   > Execute Python Script!(
    execute
    PythonScript="import unreal\r\n\r\n## BuildingProp.Medium\r\nassetObject = unreal.load_object(None, asset_object_path)\r\ndeathParticle = unreal.load_object(None, fx_object_path)\r\nresourceTable = unreal.load_object(None, resource_object_path)\r\n\r\ncdo = unreal.get_default_object(assetObject)\r\n\r\nbuildingTypeEnum = cdo.get_editor_property(\"BuildingType\")\r\nattInitKeys = cdo.get_editor_property(\"AttributeInitKeys\")\r\nattInitCategory = attInitKeys[0].get_editor_property(\"attributeInitCategory\")\r\nattachmentSlotEnum = cdo.get_editor_property(\"building_attachment_slot\")\r\ncurrentResType = cdo.get_editor_property(\"ResourceType\").name\r\ncurrentResAmount = cdo.get_editor_property(\"BuildingResourceAmountOverride\").get_editor_property(\"RowName\")\r\ncurrentAttributeCat = attInitKeys[0].get_editor_property(\"attributeInitCategory\")\r\nphysObj_PresetTag = unreal.FortniteEditorLibrary.request_gameplay_tag(\"PhysicsObject.Preset.Box\")\r\nphysObj_BuoyancyTag = unreal.FortniteEditorLibrary.request_gameplay_tag(\"PhysicsObject.BuoyancyData.Box\")\r\nresArray = [\"Wood\", \"Stone\", \"Metal\"]\r\nresStr = []\r\nfor i in resArray:\r\n    if i.lower() == str(currentResType).lower():\r\n        resStr.append(i)\r\n        break\r\n    else:\r\n        resStr.append(\"Wood\")\r\ncatArray = [\"Prop\", \"Resource\"]\r\ncatStr = []\r\nfor i in catArray:\r\n    if i.lower() in str(currentResAmount).lower():\r\n        catStr.append(i)\r\n        break\r\n    else:\r\n        catStr.append(\"Prop\")\r\ndenArray = [\"Low\", \"Medium\", \"High\", \"VeryHigh\"]\r\ndenStr = []\r\nfor i in denArray:\r\n    if i.lower() in str(currentResAmount).lower():\r\n        denStr.append(i)\r\n        break\r\n    else:\r\n        denStr.append(\"Medium\")\r\nphysObj_PhysMatTag = unreal.FortniteEditorLibrary.request_gameplay_tag(\"PhysicsObject.PhysData.\" + resStr[0])\r\n\r\nattInitKeys[0] = unreal.FortAttributeInitializationKey(\"Prop\"+str(resStr[0]), \"2\")\r\nattInitKeys[1] = unreal.FortAttributeInitializationKey(\"Prop\"+str(resStr[0]), \"2\")\r\ncdo.set_editor_property(\"BuildingType\", buildingTypeEnum.PROP)\r\ncdo.set_editor_property(\"building_attachment_slot\", attachmentSlotEnum.SLOT_FLOOR)\r\ncdo.get_editor_property(\"BuildingResourceAmountOverride\").set_editor_property(\"CurveTable\", resourceTable)\r\ncdo.get_editor_property(\"BuildingResourceAmountOverride\").set_editor_property(\"RowName\", \"Prop\"+str(resStr[0])+\"Medium\")\r\ncdo.set_editor_property(\"Physics_Object_Preset_Tag\", physObj_PresetTag)\r\ncdo.set_editor_property(\"Physics_Object_Buoyancy_Data_Tag\", physObj_BuoyancyTag)\r\ncdo.set_editor_property(\"Physics_Object_Physical_Data_Tag\", physObj_PhysMatTag)\r\ncdo.set_editor_property(\"DeathParticles\", deathParticle)\r\ncdo.set_editor_property(\"death_particle_socket_name\", \"FX_Death\")\r\ncdo.set_editor_property(\"bDoNotBlockBuildings\", True)\r\ncdo.set_editor_property(\"bDestroyOnPlayerBuildingPlacement\", True)\r\ncdo.set_editor_property(\"bIsMovable\", True)\r\ncdo.set_editor_property(\"bCreatePhysicsObjectComponent\", True)\r\ncdo.set_editor_property(\"bCanBeDamaged\", True)\r\ncdo.set_editor_property(\"bEnableAutoLodGeneration\", False)\r\n"
    asset_object_path=Append(A=ToSoftObjectPath(InAssetData=@n2), B=_C)
    fx_object_path:String=MakeLiteralString(Value=/Game/Effects/Fort_Effects/World/Destruction/Base_Effects/P_General_Deco_Destruction_Medium)
    resource_object_path:String=Get ResourceRates() -> (ResourceRates:String)
   ) -> (then, ReturnValue) {inputs [asset_object_path, fx_object_path, resource_object_path]}
  ]}
 Set_asDense {nodes [
   Set_asDense() {is=title K2Node_FunctionEntry}
   > n1 = GetSelectedAssetData()
   > n2 = For Each Loop(Array=@n1)
   > n3 = Set l_currentAssetData(l_currentAssetData:Asset Data Structure=@n2) -> (Output_Get:Asset Data Structure)
   > n4 = CheckParentClass!(execute, self, inAssetData:Asset Data Structure=@n3) -> (then, isProp:bool, isValid:bool) {is=title K2Node_CallFunction}
   > n5 = Branch(Condition=@n4.isValid)
   > n6 = Branch(Condition=@n4)
   n7 = Get l_currentAssetData() -> (l_currentAssetData:Asset Data Structure)
   > Execute Python Script!(
    execute
    PythonScript="import unreal\r\n\r\n## Set Props to a resource dense property configuration\r\nassetObject = unreal.load_object(None, asset_object_path)\r\nresourceTable = unreal.load_object(None, resource_object_path)\r\ncdo = unreal.get_default_object(assetObject)\r\n\r\nattInitKeys = cdo.get_editor_property(\"AttributeInitKeys\")\r\nattInitCategory = attInitKeys[0].get_editor_property(\"attributeInitCategory\")\r\ncurrentResType = cdo.get_editor_property(\"ResourceType\").name\r\ncurrentResAmount = cdo.get_editor_property(\"BuildingResourceAmountOverride\").get_editor_property(\"RowName\")\r\ncurrentAttributeCat = attInitKeys[0].get_editor_property(\"attributeInitCategory\")\r\ncurrentAttributeSubCat = attInitKeys[0].get_editor_property(\"attributeInitSubCategory\")\r\nphysObj_PresetTag = unreal.FortniteEditorLibrary.request_gameplay_tag(\"PhysicsObject.Preset.Box\")\r\nphysObj_BuoyancyTag = unreal.FortniteEditorLibrary.request_gameplay_tag(\"PhysicsObject.BuoyancyData.Sinker\")\r\nresArray = [\"Wood\", \"Stone\", \"Metal\"]\r\nresStr = []\r\nfor i in resArray:\r\n    if i.lower() == str(currentResType).lower():\r\n        resStr.append(i)\r\n        break\r\n    else:\r\n        resStr.append(\"Wood\")\r\ncatArray = [\"Prop\", \"Resource\"]\r\ncatStr = []\r\nfor i in catArray:\r\n    if i.lower() in str(currentResAmount).lower():\r\n        catStr.append(i)\r\n        break\r\n    else:\r\n        catStr.append(\"Prop\")\r\ndenArray = [\"Low\", \"Medium\", \"High\", \"VeryHigh\"]\r\ndenStr = []\r\nfor i in denArray:\r\n    if i.lower() in str(currentResAmount).lower():\r\n        denStr.append(i)\r\n        break\r\n    else:\r\n        denStr.append(\"Medium\")\r\n## physObj_PhysMatTag = unreal.FortniteEditorLibrary.request_gameplay_tag(\"PhysicsObject.PhysData.\"+resStr[0]+\".Dense\")\r\n\r\nattInitKeys[0] = unreal.FortAttributeInitializationKey(\"Resource\"+str(resStr[0]), currentAttributeSubCat)\r\nattInitKeys[1] = attInitKeys[0]\r\ncdo.get_editor_property(\"BuildingResourceAmountOverride\").set_editor_property(\"CurveTable\", resourceTable)\r\ncdo.get_editor_property(\"BuildingResourceAmountOverride\").set_editor_property(\"RowName\", \r\n\t\"Resource\"+str(resStr[0])+str(denStr[0]))\r\ncdo.set_editor_property(\"Physics_Object_Preset_Tag\", physObj_PresetTag)\r\ncdo.set_editor_property(\"Physics_Object_Buoyancy_Data_Tag\", physObj_BuoyancyTag)\r\n## cdo.set_editor_property(\"Physics_Object_Physical_Data_Tag\", physObj_PhysMatTag)\r\ncdo.set_editor_property(\"bIsMovable\", False)\r\n"
    asset_object_path=Append(A=ToSoftObjectPath(InAssetData=@n7), B=_C)
    resource_object_path=Get ResourceRates()
   ) -> (then, ReturnValue) {inputs [asset_object_path, resource_object_path]}
   n8 = Get l_currentAssetData()
   Execute Python Script(
    execute=@n6.else
    PythonScript="import unreal\r\n\r\n## Set Architecture to a resource dense property configuration\r\nassetObject = unreal.load_object(None, asset_object_path)\r\nresourceTable = unreal.load_object(None, resource_object_path)\r\ncdo = unreal.get_default_object(assetObject)\r\n\r\ncdo.get_editor_property(\"BuildingResourceAmountOverride\").set_editor_property(\"CurveTable\", resourceTable)\r\ncdo.get_editor_property(\"BuildingResourceAmountOverride\").set_editor_property(\"RowName\", \"LDBuildingThick\")\r\n"
    asset_object_path=Append(A=ToSoftObjectPath(InAssetData=@n8), B=_C)
    resource_object_path=Get ResourceRates()
   ) {inputs [asset_object_path, resource_object_path]}
   FortShippingLog(execute=@n5.else, InString=Unexpected Class Type, bLogAsWarning=true)
  ]}
 Set_asMetal {nodes [
   Set_asMetal() {is=title K2Node_FunctionEntry}
   > n1 = GetSelectedAssetData()
   > n2 = For Each Loop(Array=@n1)
   > Set l_currentAssetData(l_currentAssetData=@n2)
   n3 = Get l_currentAssetData()
   > Execute Python Script(
    PythonScript="import unreal\r\n\r\n## Set ResourceType to Metal\r\nassetObject = unreal.load_object(None, asset_object_path)\r\nresourceTable = unreal.load_object(None, resource_object_path)\r\n\r\ncdo = unreal.get_default_object(assetObject)\r\nresourceTypeEnum = cdo.get_editor_property(\"ResourceType\")\r\n\r\ncdo.set_editor_property(\"ResourceType\", resourceTypeEnum.METAL)\r\n"
    asset_object_path=Append(A=ToSoftObjectPath(InAssetData=@n3), B=_C)
    resource_object_path=Get ResourceRates()
   ) {inputs [asset_object_path, resource_object_path]}
   > Unify!(execute, self, inAssetData:Asset Data Structure=Get l_currentAssetData()) -> (then)
  ]}
 Set_asStone {nodes [
   Set_asStone() {is=title K2Node_FunctionEntry}
   > n1 = GetSelectedAssetData()
   > n2 = For Each Loop(Array=@n1)
   > Set l_currentAssetData(l_currentAssetData=@n2)
   n3 = Get l_currentAssetData()
   > Execute Python Script(
    PythonScript="import unreal\r\n\r\n## Set ResourceType to Stone\r\nassetObject = unreal.load_object(None, asset_object_path)\r\nresourceTable = unreal.load_object(None, resource_object_path)\r\n\r\ncdo = unreal.get_default_object(assetObject)\r\nresourceTypeEnum = cdo.get_editor_property(\"ResourceType\")\r\n\r\ncdo.set_editor_property(\"ResourceType\", resourceTypeEnum.STONE)\r\n"
    asset_object_path=Append(A=ToSoftObjectPath(InAssetData=@n3), B=_C)
    resource_object_path=Get ResourceRates()
   ) {inputs [asset_object_path, resource_object_path]}
   > Unify(inAssetData=Get l_currentAssetData())
  ]}
 Set_asWood {nodes [
   Set_asWood() {is=title K2Node_FunctionEntry}
   > n1 = GetSelectedAssetData()
   > n2 = For Each Loop(Array=@n1)
   > Set l_currentAssetData(l_currentAssetData=@n2)
   n3 = Get l_currentAssetData()
   > Execute Python Script(
    PythonScript="import unreal\r\n\r\n## Set ResourceType to Wood\r\nassetObject = unreal.load_object(None, asset_object_path)\r\nresourceTable = unreal.load_object(None, resource_object_path)\r\n\r\ncdo = unreal.get_default_object(assetObject)\r\nresourceTypeEnum = cdo.get_editor_property(\"ResourceType\")\r\n\r\ncdo.set_editor_property(\"ResourceType\", resourceTypeEnum.WOOD)\r\n"
    asset_object_path=Append(A=ToSoftObjectPath(InAssetData=@n3), B=_C)
    resource_object_path=Get ResourceRates()
   ) {inputs [asset_object_path, resource_object_path]}
   > Unify(inAssetData=Get l_currentAssetData())
  ]}
 Unify {nodes [
   n1 = Unify!() -> (then, inAssetData:Asset Data Structure) {is=title K2Node_FunctionEntry}
   > n2 = CheckParentClass(inAssetData=@n1)
   > n3 = Branch(Condition=@n2.isValid)
   > n4 = Branch(Condition=@n2)
   n5 = Get inAssetData()
   > n6 = Execute Python Script(
    PythonScript="import unreal\r\n\r\n## Aligns resource type dependant properties to match the current resource type\r\nassetObject = unreal.load_object(None, asset_object_path)\r\nresourceTable = unreal.load_object(None, resource_object_path)\r\n\r\ncdo = unreal.get_default_object(assetObject)\r\nattInitKeys = cdo.get_editor_property(\"AttributeInitKeys\")\r\nattInitCategory = attInitKeys[0].get_editor_property(\"attributeInitCategory\")\r\ncurrentResType = cdo.get_editor_property(\"ResourceType\").name\r\ncurrentResAmount = cdo.get_editor_property(\"BuildingResourceAmountOverride\").get_editor_property(\"RowName\")\r\ncurrentAttributeCat = attInitKeys[0].get_editor_property(\"attributeInitCategory\")\r\ncurrentAttributeSubCat = attInitKeys[0].get_editor_property(\"attributeInitSubCategory\")\r\nresArray = [\"Wood\", \"Stone\", \"Metal\"]\r\nfor i in resArray:\r\n    if i.lower() == str(currentResType).lower():\r\n        resStr = i\r\n        break\r\n    else:\r\n        resStr = \"Wood\"\r\ncatArray = [\"Prop\", \"Resource\"]\r\nfor i in catArray:\r\n    if i.lower() in str(currentResAmount).lower():\r\n        catStr = i\r\n        break\r\n    else:\r\n        catStr = \"Prop\"\r\ndenArray = [\"Low\", \"Medium\", \"High\", \"VeryHigh\"]\r\nfor i in denArray:\r\n    if i.lower() in str(currentResAmount).lower():\r\n        denStr = i\r\n        break\r\n    else:\r\n        denStr = \"Medium\"\r\nphysObj_PhysMatTag = unreal.FortniteEditorLibrary.request_gameplay_tag(\"PhysicsObject.PhysData.\"+resStr)\r\n\r\nattInitKeys[0] = unreal.FortAttributeInitializationKey(catStr+resStr, currentAttributeSubCat)\r\nattInitKeys[1] = attInitKeys[0]\r\ncdo.get_editor_property(\"BuildingResourceAmountOverride\").set_editor_property(\"CurveTable\", resourceTable)\r\ncdo.get_editor_property(\"BuildingResourceAmountOverride\").set_editor_property(\"RowName\", catStr+resStr+denStr)\r\ncdo.set_editor_property(\"Physics_Object_Physical_Data_Tag\", physObj_PhysMatTag)\r\n"
    asset_object_path=Append(A=ToSoftObjectPath(InAssetData=@n5), B=_C)
    resource_object_path=Get ResourceRates()
   ) {inputs [asset_object_path, resource_object_path]}
   n7 = Get inAssetData()
   SaveAsset(execute=[@n6.then, @n9.then], AssetToSave=GetBaseFilename(InPath=ToSoftObjectPath(InAssetData=@n7), bRemovePath=false), bOnlyIfIsDirty=true)
   n8 = Get inAssetData()
   n9 = Execute Python Script(
    execute=@n4.else
    PythonScript="import unreal\r\n\r\n## Set Architecture to a resource dense property configuration\r\nassetObject = unreal.load_object(None, asset_object_path)\r\nresourceTable = unreal.load_object(None, resource_object_path)\r\n\r\ncdo = unreal.get_default_object(assetObject)\r\nattInitKeys = cdo.get_editor_property(\"AttributeInitKeys\")\r\ncurrentResType = cdo.get_editor_property(\"ResourceType\").name\r\ncurrentResAmount = cdo.get_editor_property(\"BuildingResourceAmountOverride\").get_editor_property(\"RowName\")\r\ncurrentAttributeCat = attInitKeys[0].get_editor_property(\"attributeInitCategory\")\r\ncurrentAttributeSubCat = attInitKeys[0].get_editor_property(\"attributeInitSubCategory\")\r\nresArray = [\"Wood\", \"Stone\", \"Metal\"]\r\nfor i in resArray:\r\n    if i.lower() == str(currentResType).lower():\r\n        resStr = i\r\n        break\r\n    else:\r\n        resStr = \"Wood\"\r\n\r\ncdo.get_editor_property(\"BuildingResourceAmountOverride\").set_editor_property(\"CurveTable\", resourceTable)\r\ncdo.get_editor_property(\"BuildingResourceAmountOverride\").set_editor_property(\"RowName\", currentResAmount)\r\nattInitKeys[0] = unreal.FortAttributeInitializationKey(\"Building\"+resStr, currentAttributeSubCat)\r\nattInitKeys[1] = attInitKeys[0]\r\n"
    asset_object_path=Append(A=ToSoftObjectPath(InAssetData=@n8), B=_C)
    resource_object_path=Get ResourceRates()
   ) {inputs [asset_object_path, resource_object_path]}
   FortShippingLog(execute=@n3.else, InString=Unexpected Class Type, bLogAsWarning=true)
  ]}
 UnifyResources {nodes [
   UnifyResources() {is=title K2Node_FunctionEntry}
   > n1 = GetSelectedAssetData()
   > n2 = For Each Loop(Array=@n1)
   > Unify(inAssetData=@n2) {is=title K2Node_CallFunction}
  ]}
]
variables [
 architectureClassWhitelist {
  type="array<Class>"
  default [
   &BuildingWall
   &BuildingFloor
   &BuildingStairs
   &BuildingRoof
   /Script/FortniteGame.BuildingPillar
   /Script/FortniteGame.BuildingPropWall
  ]
 }
 propClassWhitelist {type="array<Class>", default [&BuildingProp, /Script/FortniteGame.BuildingContainer, /Script/FortniteGame.BuildingDeco]}
 ResourceRates {type=string, default=/Game/Athena/Balance/DataTables/AthenaResourceRates}
]
